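# Physics asset for the RadRat medium male player body cosmetic, covering only the pelvis, thighs, a belt strap and a belt hook.
# Pelvis and thigh capsules form the body volumes, with the thighs constrained to the pelvis.
# Small base capsules attach to the pelvis with free swing and twist, rooting a three-link belt chain and a single-link hook.
# These dyn_ links have tight swing limits and little or no twist.
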PhysicsAsset /BRCosmetics/Characters/Player/Male/Medium/Bodies/M_MED_RadRat/Meshes/M_MED_RadRat_Physics
bodies {
 base_beltHook_01 [{shapeType=Capsule, length=0.505, radius=0.505, shapeName=base_beltHook_01_capsule, rotation=<rot 0 -90 0>}]
 base_belt__01 [{shapeType=Capsule, length=0.505, radius=0.505, shapeName=base_belt__01_capsule, rotation=<rot 0 -90 0>}]
 dyn_beltHook_01 [{shapeType=Capsule, length=2.3255, radius=1.832, shapeName=dyn_beltHook_01_capsule, center=<2 0 1.5>, rotation=<rot 90 0 0>}]
 dyn_belt__01 [{shapeType=Capsule, length=4, radius=2, shapeName=dyn_belt__01_capsule, center=<2 0 0.5>, rotation=<rot 70 0 -10>}]
 dyn_belt__02 [{shapeType=Capsule, length=4, radius=2, shapeName=dyn_belt__02_capsule, center=<4 0 0>, rotation=<rot 80 0 -10>}]
 dyn_belt__03 [{shapeType=Capsule, length=4, radius=2, shapeName=dyn_belt__03_capsule, center=<4 0 0>, rotation=<rot 80 0 10>}]
 pelvis [{shapeType=Capsule, length=5.6433, radius=9.5862, center=<0 -2.2764 0>}]
 thigh_l [{shapeType=Capsule, length=33.4879, radius=7.1336, center=<-17.6298 -0.2541 1.1847>, rotation=<rot 0.0635 -90.7684 -90.2679>}]
 thigh_r [{shapeType=Capsule, length=32.9752, radius=7.6463, center=<17.595 0.4139 -2.3277>, rotation=<rot 0.0634 -90.7683 91.8472>}]
}
constraints [
 {bone1Name=thigh_l, bone2Name=pelvis, swing1LimitDegrees=45, swing2LimitDegrees=45, twistLimitDegrees=45}
 {bone1Name=thigh_r, bone2Name=pelvis, swing1LimitDegrees=45, swing2LimitDegrees=45, twistLimitDegrees=45}
 {
  bone1Name=base_belt__01
  bone2Name=pelvis
  swing1LimitDegrees=45
  swing1Motion=Free
  swing2LimitDegrees=45
  swing2Motion=Free
  twistLimitDegrees=45
  twistMotion=Free
 }
 {bone1Name=dyn_belt__01, bone2Name=base_belt__01, swing1LimitDegrees=10, swing2LimitDegrees=10, twistLimitDegrees=2}
 {bone1Name=dyn_belt__02, bone2Name=dyn_belt__01, swing1LimitDegrees=10, swing2LimitDegrees=10, twistLimitDegrees=2}
 {bone1Name=dyn_belt__03, bone2Name=dyn_belt__02, swing1LimitDegrees=10, swing2LimitDegrees=10, twistLimitDegrees=2}
 {
  bone1Name=base_beltHook_01
  bone2Name=pelvis
  swing1LimitDegrees=45
  swing1Motion=Free
  swing2LimitDegrees=45
  swing2Motion=Free
  twistLimitDegrees=45
  twistMotion=Free
 }
 {bone1Name=dyn_beltHook_01, bone2Name=base_beltHook_01, swing1LimitDegrees=10, swing2LimitDegrees=20, twistLimitDegrees=0}
]
