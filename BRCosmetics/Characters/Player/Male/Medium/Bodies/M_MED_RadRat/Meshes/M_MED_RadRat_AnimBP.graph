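# Post-process style AnimBP for the male RadRat cosmetic body, adding secondary motion to a pose copied from the parent mesh.
# A Control Rig, aware of front-end state, feeds a cached pose into two pelvis-space RigidBody simulations.
# Crouching selects the downward-gravity simulation over the zero-gravity one, then chained spring AnimDynamics nodes swing four pant-cuff bones.
# Each update, a parameter-asset function stores gravity override and rigid-body settings in variables.
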AnimBlueprint /BRCosmetics/Characters/Player/Male/Medium/Bodies/M_MED_RadRat/Meshes/M_MED_RadRat_AnimBP
target_skeleton=&/M_MED_RadRat_Skeleton
graphs [
 AnimGraph {nodes [
   n1 = RigidBody!(
    ComponentPose=Local To Component(LocalPose=Use cached pose 'InputPose'())
    Alpha
    OverrideWorldGravity:Vector="0,0,0"
    ExternalForce
    SimSpaceSettings:Sim Space Settings Structure=(WorldAlpha=0,VelocityScaleZ=1,DampingAlpha=1,MaxLinearVelocity=10000,MaxAngularVelocity=10000,MaxLinearAcceleration=10000,MaxAngularAcceleration=10000,ExternalLinearDragV=(X=0,Y=0,Z=0),ExternalLinearVelocity=(X=0,Y=0,Z=0),ExternalAngularVelocity=(X=0,Y=0,Z=0))
   ) -> (Pose) {
    bOverrideWorldGravity=true
    overridePhysicsAsset=&/M_MED_RadRat_Physics
    simulationSpace=BaseBoneSpace
    baseBoneRef {boneName=pelvis}
   }
   n2 = Component To Local(ComponentPose=@n1)
   n3 = RigidBody!(ComponentPose=Local To Component(LocalPose=Use cached pose 'InputPose'()), Alpha, OverrideWorldGravity="0,0,-450", ExternalForce) -> (Pose) {
    bOverrideWorldGravity=true
    overridePhysicsAsset=&M_MED_RadRat_Physics
    simulationSpace=BaseBoneSpace
    baseBoneRef {boneName=pelvis}
    componentAppliedLinearAccClamp=<2000 2000 2000>
   }
   n4 = Component To Local(ComponentPose=@n3)
   n5 = Get bIsCrouching()
   n6 = Anim Dynamics - Bone: dyn_pantCuff_fr_01_l!(
    ComponentPose=Local To Component(LocalPose=Blend Poses by bool(BlendPose_0=@n4, BlendPose_1=@n2, bActiveValue=@n5))
    Alpha=1
    GravityOverride:Vector="0,0,-980"
    ExternalForce="0,0,0"
   ) -> (Pose) {
    angularDampingOverride=1
    angularSpringConstant=15000
    bAngularSpring=true
    bOverrideAngularDamping=true
    bUseGravityOverride=true
    gravityScale=0.35
    simulationSpace=BoneRelative
    alphaBoolBlend {blendOption=Linear}
    boundBone {boneName=dyn_pantCuff_fr_01_l}
    componentAppliedLinearAccClamp=<2000 2000 2000>
    componentLinearAccScale=<1 1 1>
    componentLinearVelScale=<1 1 1>
    gravityOverride {z=-980}
    physicsBodyDefinitions [{
      boundBone {boneName=dyn_pantCuff_fr_01_l}
      boxExtents=<20 20 20>
      constraintSetup {angularLimitsMax=<10 5 35>, angularLimitsMin=<-10 -30 -35>, angularTarget=<1 0 0>}
      localJointOffset=<10 0 0>
     }]
    relativeSpaceBone {boneName=root}
   }
   n7 = Anim Dynamics - Bone: dyn_pantCuff_bk_01_l!(ComponentPose=@n6, Alpha=1, GravityOverride:Vector="0,0,-980", ExternalForce="0,0,0") -> (Pose) {
    angularDampingOverride=1
    angularSpringConstant=15000
    bAngularSpring=true
    bOverrideAngularDamping=true
    bUseGravityOverride=true
    gravityScale=0.35
    simulationSpace=BoneRelative
    alphaBoolBlend {blendOption=Linear}
    boundBone {boneName=dyn_pantCuff_bk_01_l}
    componentAppliedLinearAccClamp=<2000 2000 2000>
    componentLinearAccScale=<1 1 1>
    componentLinearVelScale=<1 1 1>
    gravityOverride {z=-980}
    physicsBodyDefinitions [{
      boundBone {boneName=dyn_pantCuff_bk_01_l}
      boxExtents=<20 30 30>
      constraintSetup {angularLimitsMax=<10 2 35>, angularLimitsMin=<-10 -30 -35>, angularTarget=<1 0 0>}
      localJointOffset=<10 0 0>
     }]
    relativeSpaceBone {boneName=root}
   }
   n8 = Anim Dynamics - Bone: dyn_pantCuff_fr_01_r!(ComponentPose=@n7, Alpha=1, GravityOverride:Vector="0,0,-980", ExternalForce="0,0,0") -> (Pose) {
    angularDampingOverride=1
    angularSpringConstant=15000
    bAngularSpring=true
    bOverrideAngularDamping=true
    bUseGravityOverride=true
    gravityScale=0.35
    simulationSpace=BoneRelative
    alphaBoolBlend {blendOption=Linear}
    boundBone {boneName=dyn_pantCuff_fr_01_r}
    componentAppliedLinearAccClamp=<2000 2000 2000>
    componentLinearAccScale=<1 1 1>
    componentLinearVelScale=<1 1 1>
    gravityOverride {z=-980}
    physicsBodyDefinitions [{
      boundBone {boneName=dyn_pantCuff_fr_01_r}
      boxExtents=<20 20 20>
      constraintSetup {angularLimitsMax=<10 5 35>, angularLimitsMin=<-10 -30 -35>, angularTarget=<1 0 0>}
      localJointOffset=<10 0 0>
     }]
    relativeSpaceBone {boneName=root}
   }
   n9 = Anim Dynamics - Bone: dyn_pantCuff_bk_01_r!(ComponentPose=@n8, Alpha=1, GravityOverride:Vector="0,0,-980", ExternalForce="0,0,0") -> (Pose) {
    angularDampingOverride=1
    angularSpringConstant=15000
    bAngularSpring=true
    bOverrideAngularDamping=true
    bUseGravityOverride=true
    gravityScale=0.35
    simulationSpace=BoneRelative
    alphaBoolBlend {blendOption=Linear}
    boundBone {boneName=dyn_pantCuff_bk_01_r}
    componentAppliedLinearAccClamp=<2000 2000 2000>
    componentLinearAccScale=<1 1 1>
    componentLinearVelScale=<1 1 1>
    gravityOverride {z=-980}
    physicsBodyDefinitions [{
      boundBone {boneName=dyn_pantCuff_bk_01_r}
      boxExtents=<20 30 30>
      constraintSetup {angularLimitsMax=<10 10 35>, angularLimitsMin=<-10 -30 -35>, angularTarget=<1 0 0>}
      localJointOffset=<10 0 0>
     }]
    relativeSpaceBone {boneName=root}
   }
   Output Pose(Result=Component To Local(ComponentPose=@n9)) {layerGroup=None}
   n10 = Control Rig!(Source=Copy Pose From Mesh() {bUseAttachedParent=true}, Alpha, FrontEnd:bool=Get bIsInFrontEnd()) -> (Pose) {controlRigAssetReference {blueprintRigClass=&/M_MED_RadRat_CtrlRig.M_MED_RadRat_CtrlRig_C}}
   Save cached pose 'InputPose'(Pose=@n10) {cacheName=InputPose}
  ]}
 EventGraph {nodes [
   Event BlueprintUpdateAnimation()
   > Sequence()
   > n1 = "Gravity Override State Function (Parameter Asset) (Nativized)"(
    CharacterPartAnimInstance=Self-Reference()
    GameTimeInSeconds=GetGameTimeInSeconds()
    Parameters=&/M_MED_RadRat_Parameters
   )
   > Set GravityOverride(GravityOverride=@n1)
   > Set RBAN Settings Out(RBAN Settings Out=@n1.OutRigidBodyAnimNodeParameters)
  ]}
]
variables [
 GravityOverride {type=Vector}
 RBAN Settings Out {type=FortRigidBodyAnimNodeParameters}
]
